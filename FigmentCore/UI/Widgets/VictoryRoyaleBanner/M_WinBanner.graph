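# M_WinBanner is a translucent UI-domain material for the Victory Royale banner widget.
# It contains a texture sample of T_UI_winbanner using the alpha sampler type, alongside a constant of one.
# The graph lists no links from these two nodes to material outputs, and it has no parameters or animation.
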Material /FigmentCore/UI/Widgets/VictoryRoyaleBanner/M_WinBanner
graph {
 properties {bAutomaticallySetUsageInEditor=false, blendMode=BLEND_Translucent, materialDomain=MD_UI, usedShadingModels=Default Lit}
 nodes [
  Constant() {r=1}
  TextureSample() {samplerType=SAMPLERTYPE_Alpha, texture=&/T_UI_winbanner}
 ]
}
