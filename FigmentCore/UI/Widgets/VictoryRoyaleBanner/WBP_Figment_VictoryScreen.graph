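# Figment post-game placement overlay, derived from PostGamePlacementOverlay, that shows a win, tie or loss banner.
# Placement events, mirrored by custom Show Win, Tie and Loss events, switch TopSwitcher between overlays and reveal Overlay_Content.
# Win displays the Victory Royale banner and sets PlacementVictory on the Athena HUD context; loss fills in the local player's placement from the Fort HUD context.
# PreConstruct tints the banner image from winbanner_Color.
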WidgetBlueprint /FigmentCore/UI/Widgets/VictoryRoyaleBanner/WBP_Figment_VictoryScreen
graphs [EventGraph {nodes [
   n1 = Show Tie() {is=title K2Node_CustomEvent}
   SetActiveWidget(
    execute=[@n1.then, @n6.then]
    self:Widget Switcher Object Reference=Get TopSwitcher() -> (TopSwitcher:Widget Switcher Object Reference)
    Widget=Get TieOverlay() -> (TieOverlay:Overlay Object Reference)
   )
   > SetVisibility(self=Get Overlay_Content() -> (Overlay_Content:Overlay Object Reference), InVisibility=HitTestInvisible)
   n2 = Show Win() {is=title K2Node_CustomEvent}
   Sequence(execute=[@n2.then, @n5.then])
   > SetActiveWidget(self=Get TopSwitcher(), Widget=Get FigmentVictoryOverlay() -> (FigmentVictoryOverlay:Overlay Object Reference))
   > SetPostGameWidgetPlacement!(
    execute
    self:Athena HUDContext Object Reference=Get AthenaHUDContext!() -> (ReturnValue:Athena HUDContext Object Reference) {is=title K2Node_GetBlueprintContextEx}
    WidgetPlacement:Name=PlacementVictory
   ) -> (then)
   > SetVisibility(self=Get Overlay_Content(), InVisibility=HitTestInvisible)
   n3 = Show Loss() {is=title K2Node_CustomEvent}
   n4 = Get FortHUDContext!() -> (ReturnValue:Fort HUDContext Object Reference) {is=title K2Node_GetBlueprintContextEx}
   "SetText (Text)"(
    execute=[@n3.then, @n7.then]
    self=Get Text_Placement() -> (Text_Placement:Common Text Object Reference)
    InText=Format Text!(Format="LOCTEXT(\"##{0}\")", 0:int=GetLocalPlayerPlacement!(self:Fort HUDContext Object Reference=@n4) -> (ReturnValue:int=0)) -> (Result)
   )
   > SetActiveWidget(self=Get TopSwitcher(), Widget=Get LosingOverlay() -> (LosingOverlay:Overlay Object Reference))
   > SetVisibility(self=Get Overlay_Content(), InVisibility=HitTestInvisible)
   Event PreConstruct()
   > Set ColorAndOpacity!(
    execute
    ColorAndOpacity:Linear Color Structure=Get winbanner_Color() -> (winbanner_Color:Linear Color Structure)
    self:Image Object Reference=Get Image_Overlay() -> (Image_Overlay:Image Object Reference)
   ) -> (then, Output_Get:Linear Color Structure)
   n5 = Event ShowWinningPlacement() {is=title K2Node_Event}
   n6 = Event ShowTiePlacement() {is=title K2Node_Event}
   n7 = Event ShowLosingPlacement() {is=title K2Node_Event}
   Event Tick()
  ]}]
variables [winbanner_Color {type=LinearColor, default=<rgba 0.0627 0.2157 0.4784 1>}]
widgets {
 info {inheritedWidgetCount=1, parentClass=/Script/FortniteUI.PostGamePlacementOverlay, rootWidgetClass=&Overlay, widgetCount=18}
 widgets [
  Overlay_Content {widgetClassPath=&Overlay, bIsVariable=true, widget=Overlay_Content}
  TopSwitcher {
   widgetClassPath=&WidgetSwitcher
   bIsVariable=true
   parent=Overlay_Content
   widget=TopSwitcher
   properties {visibility=HitTestInvisible}
   slot {class=OverlaySlot, properties {horizontalAlignment=HAlign_Fill, verticalAlignment=VAlign_Fill}}
  }
  LosingOverlay {
   widgetClassPath=&Overlay
   bIsVariable=true
   parent=TopSwitcher
   widget=LosingOverlay
   properties {flowDirectionPreference=Culture}
   slot {class=WidgetSwitcherSlot, properties {horizontalAlignment=HAlign_Center, verticalAlignment=VAlign_Top, padding=<ltrb 0 85 0 0>}}
  }
  HorizontalBox-Placement {
   widgetClassPath=&HorizontalBox
   parent=LosingOverlay
   widget=HorizontalBox-Placement
   properties {renderTransformPivot=<0.5 0>}
   slot {class=OverlaySlot}
  }
  CommonTextBlock_24 {
   widgetClassPath=&CommonTextBlock
   parent=HorizontalBox-Placement
   widget=CommonTextBlock_24
   properties {
    text=You Placed
    textTransformPolicy=ToUpper
    font {
     bForceMonospaced=false
     bMaterialIsStencil=false
     fontMaterial=null
     fontObject=&Burbank
     letterSpacing=0
     monospacedWidth=1
     size=48
     skewAmount=0
     typefaceFontName=Burbank-Black
     outlineSettings {
      bApplyOutlineToDropShadows=false
      bMiteredCorners=false
      bSeparateFillAlpha=false
      outlineMaterial=null
      outlineSize=0
      outlineColor=<rgba 0 0 0 1>
     }
    }
    renderTransform {angle=0, scale=<1 1>, shear=<-12 0>, translation=<0 0>}
    shadowColorAndOpacity=<rgba 0.0052 0.0052 0.0056 0.8>
    shadowOffset=<4 4>
   }
   slot {class=HorizontalBoxSlot, properties {verticalAlignment=VAlign_Center, padding=<ltrb 0 0 10 0>}}
  }
  Text_Placement {
   widgetClassPath=&CommonTextBlock
   bIsVariable=true
   parent=HorizontalBox-Placement
   widget=Text_Placement
   properties {
    text="##99"
    font {
     bForceMonospaced=false
     bMaterialIsStencil=false
     fontMaterial=null
     fontObject=&Burbank
     letterSpacing=0
     monospacedWidth=1
     size=86
     skewAmount=0
     typefaceFontName=Burbank-Black
     outlineSettings {
      bApplyOutlineToDropShadows=false
      bMiteredCorners=false
      bSeparateFillAlpha=false
      outlineMaterial=null
      outlineSize=0
      outlineColor=<rgba 0 0 0 1>
     }
    }
    renderTransform {angle=7, scale=<1 1>, shear=<0 0>, translation=<0 0>}
    shadowColorAndOpacity=<rgba 0.0052 0.0052 0.0056 0.8>
    shadowOffset=<4 4>
   }
   slot {class=HorizontalBoxSlot, properties {verticalAlignment=VAlign_Center}}
  }
  TieOverlay {
   widgetClassPath=&Overlay
   bIsVariable=true
   parent=TopSwitcher
   widget=TieOverlay
   slot {class=WidgetSwitcherSlot, properties {horizontalAlignment=HAlign_Center, verticalAlignment=VAlign_Top, padding=<ltrb 0 85 0 0>}}
  }
  Text_Draw {
   widgetClassPath=&CommonTextBlock
   bIsVariable=true
   parent=TieOverlay
   widget=Text_Draw
   properties {
    text="DRAW!"
    textTransformPolicy=ToUpper
    font {
     bForceMonospaced=false
     bMaterialIsStencil=false
     fontMaterial=null
     fontObject=&Burbank
     letterSpacing=0
     monospacedWidth=1
     size=120
     skewAmount=0
     typefaceFontName=Burbank-Black
     outlineSettings {
      bApplyOutlineToDropShadows=false
      bMiteredCorners=false
      bSeparateFillAlpha=false
      outlineMaterial=null
      outlineSize=0
      outlineColor=<rgba 0 0 0 1>
     }
    }
    renderTransform {angle=0, scale=<1 1>, shear=<-12 0>, translation=<0 0>}
    shadowColorAndOpacity=<rgba 0.0052 0.0052 0.0056 0.8>
    shadowOffset=<4 4>
   }
   slot {class=OverlaySlot, properties {horizontalAlignment=HAlign_Fill, verticalAlignment=VAlign_Fill}}
  }
  FigmentVictoryOverlay {
   widgetClassPath=&Overlay
   bIsVariable=true
   parent=TopSwitcher
   widget=FigmentVictoryOverlay
   slot {class=WidgetSwitcherSlot, properties {horizontalAlignment=HAlign_Center, verticalAlignment=VAlign_Top, padding=<ltrb 0 100 0 0>}}
  }
  Overlay_BGArt {
   widgetClassPath=&Overlay
   parent=FigmentVictoryOverlay
   widget=Overlay_BGArt
   slot {class=OverlaySlot, properties {horizontalAlignment=HAlign_Fill, verticalAlignment=VAlign_Center}}
  }
  Image_Overlay {
   widgetClassPath=&Image
   bIsVariable=true
   parent=Overlay_BGArt
   widget=Image_Overlay
   properties {
    brush {
     bIsDynamicallyLoaded=false
     drawAs=Image
     imageType=FullColor
     mirroring=NoMirror
     resourceName=None
     resourceObject=&/M_WinBanner
     tiling=NoTile
     imageSize=<867 121>
     margin=<ltrb 0 0 0 0>
     outlineSettings {
      bUseBrushTransparency=false
      roundingType=HalfHeightRadius
      width=0
      color {colorUseRule=UseColor_Specified, specifiedColor=<rgba 0 0 0 0>}
      cornerRadii=<0 0 0 0>
     }
     tintColor {colorUseRule=UseColor_Specified, specifiedColor=<rgba 1 1 1 1>}
     uVRegion {bIsValid=false, max=<0 0>, min=<0 0>}
    }
    colorAndOpacity=<rgba 1 0.0963 0 1>
    renderTransform {angle=0, scale=<1 1>, shear=<0 0>, translation=<-1 0>}
   }
   slot {class=OverlaySlot, properties {horizontalAlignment=HAlign_Center}}
  }
  NamedSlot_Crown {
   widgetClassPath=&NamedSlot
   bIsVariable=true
   parent=FigmentVictoryOverlay
   widget=NamedSlot_Crown
   properties {renderTransform {angle=0, scale=<1 1>, shear=<0 0>, translation=<0 17>}}
   slot {class=OverlaySlot, properties {horizontalAlignment=HAlign_Center}}
  }
  HorizontalBox_301 {
   widgetClassPath=&HorizontalBox
   parent=FigmentVictoryOverlay
   widget=HorizontalBox_301
   slot {class=OverlaySlot, properties {horizontalAlignment=HAlign_Center, verticalAlignment=VAlign_Center}}
  }
  CommonVisualAttachment_135 {
   widgetClassPath=&CommonVisualAttachment
   parent=HorizontalBox_301
   widget=CommonVisualAttachment_135
   properties {contentAnchor=<1 0.45>}
   slot {class=HorizontalBoxSlot, properties {horizontalAlignment=HAlign_Center, verticalAlignment=VAlign_Center}}
  }
  RichTextBlock {
   widgetClassPath=/Script/UMG.RichTextBlock
   parent=CommonVisualAttachment_135
   widget=RichTextBlock
   properties {
    bOverrideDefaultStyle=true
    justification=Center
    text="##1"
    textStyleSet=/Game/Athena/HUD/EndGame/VictoryRichTextStyles
    defaultTextStyleOverride {
     overflowPolicy=Clip
     transformPolicy=None
     colorAndOpacity {colorUseRule=UseColor_Specified, specifiedColor=<rgba 0.9087 1 0.1834 1>}
     font {
      bForceMonospaced=false
      bMaterialIsStencil=false
      fontMaterial=null
      fontObject=&Burbank
      letterSpacing=0
      monospacedWidth=1
      size=105
      skewAmount=0
      typefaceFontName=Burbank-Black
      outlineSettings {
       bApplyOutlineToDropShadows=false
       bMiteredCorners=false
       bSeparateFillAlpha=false
       outlineMaterial=null
       outlineSize=0
       outlineColor=<rgba 0 0 0 1>
      }
     }
     highlightColor {colorUseRule=UseColor_Specified, specifiedColor=<rgba 0 0 0 1>}
     highlightShape {
      bIsDynamicallyLoaded=false
      drawAs=Image
      imageType=NoImage
      mirroring=NoMirror
      resourceName=None
      resourceObject=null
      tiling=NoTile
      imageSize=<32 32>
      margin=<ltrb 0 0 0 0>
      outlineSettings {
       bUseBrushTransparency=false
       roundingType=HalfHeightRadius
       width=0
       color {colorUseRule=UseColor_Specified, specifiedColor=<rgba 0 0 0 0>}
       cornerRadii=<0 0 0 0>
      }
      tintColor {colorUseRule=UseColor_Specified, specifiedColor=<rgba 1 1 1 1>}
      uVRegion {bIsValid=false, max=<0 0>, min=<0 0>}
     }
     selectedBackgroundColor {colorUseRule=UseColor_Foreground, specifiedColor=<rgba 1 0 1 1>}
     shadowColorAndOpacity=<rgba 0 0 0 0.594>
     shadowOffset=<1.6724 3.1534>
     strikeBrush {
      bIsDynamicallyLoaded=false
      drawAs=Image
      imageType=NoImage
      mirroring=NoMirror
      resourceName=None
      resourceObject=null
      tiling=NoTile
      imageSize=<32 32>
      margin=<ltrb 0 0 0 0>
      outlineSettings {
       bUseBrushTransparency=false
       roundingType=HalfHeightRadius
       width=0
       color {colorUseRule=UseColor_Specified, specifiedColor=<rgba 0 0 0 0>}
       cornerRadii=<0 0 0 0>
      }
      tintColor {colorUseRule=UseColor_Specified, specifiedColor=<rgba 1 1 1 1>}
      uVRegion {bIsValid=false, max=<0 0>, min=<0 0>}
     }
     underlineBrush {
      bIsDynamicallyLoaded=false
      drawAs=Image
      imageType=NoImage
      mirroring=NoMirror
      resourceName=None
      resourceObject=null
      tiling=NoTile
      imageSize=<32 32>
      margin=<ltrb 0 0 0 0>
      outlineSettings {
       bUseBrushTransparency=false
       roundingType=HalfHeightRadius
       width=0
       color {colorUseRule=UseColor_Specified, specifiedColor=<rgba 0 0 0 0>}
       cornerRadii=<0 0 0 0>
      }
      tintColor {colorUseRule=UseColor_Specified, specifiedColor=<rgba 1 1 1 1>}
      uVRegion {bIsValid=false, max=<0 0>, min=<0 0>}
     }
    }
    renderTransform {angle=-10, scale=<1.33 1.33>, shear=<0 0>, translation=<-50 0>}
   }
   slot {class=SizeBoxSlot, properties {horizontalAlignment=HAlign_Left, verticalAlignment=VAlign_Center}}
  }
  SizeBox_280 {
   widgetClassPath=&SizeBox
   parent=HorizontalBox_301
   widget=SizeBox_280
   properties {bOverride_MinDesiredWidth=true, minDesiredWidth=500}
   slot {class=HorizontalBoxSlot, properties {horizontalAlignment=HAlign_Left, verticalAlignment=VAlign_Top}}
  }
  RichTextBlock_104 {
   widgetClassPath=&RichTextBlock
   parent=SizeBox_280
   widget=RichTextBlock_104
   properties {
    bOverrideDefaultStyle=true
    justification=Center
    text="Victory Royale!"
    textStyleSet=&VictoryRichTextStyles
    defaultTextStyleOverride {
     overflowPolicy=Clip
     transformPolicy=None
     colorAndOpacity {colorUseRule=UseColor_Specified, specifiedColor=<rgba 1 1 1 1>}
     font {
      bForceMonospaced=false
      bMaterialIsStencil=false
      fontMaterial=null
      fontObject=&Burbank
      letterSpacing=0
      monospacedWidth=1
      size=56.25
      skewAmount=0
      typefaceFontName=Burbank-Black
      outlineSettings {
       bApplyOutlineToDropShadows=false
       bMiteredCorners=false
       bSeparateFillAlpha=false
       outlineMaterial=null
       outlineSize=0
       outlineColor=<rgba 0 0 0 1>
      }
     }
     highlightColor {colorUseRule=UseColor_Specified, specifiedColor=<rgba 0 0 0 1>}
     highlightShape {
      bIsDynamicallyLoaded=false
      drawAs=Image
      imageType=NoImage
      mirroring=NoMirror
      resourceName=None
      resourceObject=null
      tiling=NoTile
      imageSize=<32 32>
      margin=<ltrb 0 0 0 0>
      outlineSettings {
       bUseBrushTransparency=false
       roundingType=HalfHeightRadius
       width=0
       color {colorUseRule=UseColor_Specified, specifiedColor=<rgba 0 0 0 0>}
       cornerRadii=<0 0 0 0>
      }
      tintColor {colorUseRule=UseColor_Specified, specifiedColor=<rgba 1 1 1 1>}
      uVRegion {bIsValid=false, max=<0 0>, min=<0 0>}
     }
     selectedBackgroundColor {colorUseRule=UseColor_Foreground, specifiedColor=<rgba 1 0 1 1>}
     shadowColorAndOpacity=<rgba 0 0 0 1>
     shadowOffset=<0 0>
     strikeBrush {
      bIsDynamicallyLoaded=false
      drawAs=Image
      imageType=NoImage
      mirroring=NoMirror
      resourceName=None
      resourceObject=null
      tiling=NoTile
      imageSize=<32 32>
      margin=<ltrb 0 0 0 0>
      outlineSettings {
       bUseBrushTransparency=false
       roundingType=HalfHeightRadius
       width=0
       color {colorUseRule=UseColor_Specified, specifiedColor=<rgba 0 0 0 0>}
       cornerRadii=<0 0 0 0>
      }
      tintColor {colorUseRule=UseColor_Specified, specifiedColor=<rgba 1 1 1 1>}
      uVRegion {bIsValid=false, max=<0 0>, min=<0 0>}
     }
     underlineBrush {
      bIsDynamicallyLoaded=false
      drawAs=Image
      imageType=NoImage
      mirroring=NoMirror
      resourceName=None
      resourceObject=null
      tiling=NoTile
      imageSize=<32 32>
      margin=<ltrb 0 0 0 0>
      outlineSettings {
       bUseBrushTransparency=false
       roundingType=HalfHeightRadius
       width=0
       color {colorUseRule=UseColor_Specified, specifiedColor=<rgba 0 0 0 0>}
       cornerRadii=<0 0 0 0>
      }
      tintColor {colorUseRule=UseColor_Specified, specifiedColor=<rgba 1 1 1 1>}
      uVRegion {bIsValid=false, max=<0 0>, min=<0 0>}
     }
    }
   }
   slot {class=SizeBoxSlot, properties {horizontalAlignment=HAlign_Center, verticalAlignment=VAlign_Center}}
  }
  TutorialHighlight {widgetClassPath=&TutorialHighlight}
 ]
}
